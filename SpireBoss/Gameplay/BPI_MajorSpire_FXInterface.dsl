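; Blueprint interface that exposes Major Spire boss state to its FX code.
; It declares getters for the beam stop transform, the hit actor and normal at that stop, and flags for blocking and water blocking hits.
; It also provides getters for the line trace ignored actors and a budgeted skeletal mesh component.
; All functions are bare declarations with default returns, so implementers supply the values.
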(fn GetBeamStopTransform
  (return 0))

(fn GetFoundBlockingHit
  (return false))

(fn GetFoundWaterBlockingHit
  (return false))

(fn GetLineTraceIgnoredActors
  (return 0))

(fn GetSkeletalMeshComponentBudgeted
  (return 0))

(fn GetStopHitActor
  (return 0))

(fn GetStopHitNormal
  (return "0, 0, 0"))
